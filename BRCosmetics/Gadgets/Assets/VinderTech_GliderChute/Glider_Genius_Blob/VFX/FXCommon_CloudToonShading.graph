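# Toon cloud shading material function, in the Genius Blob glider VFX folder, that outputs a stylized Emissive color from pixel normal, light direction and masks.
# A view-based rim term blends lit to rim color, then flattened-normal lighting terms add shadow color, multiply in soft shadow and blend toward fake GI color.
# Each lighting term has its own flatten, offset, multiply, power and remap parameters.
# Masks RGBA scale rim, shadow, soft shadow and fake GI.
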MaterialFunction /BRCosmetics/Gadgets/Assets/VinderTech_GliderChute/Glider_Genius_Blob/VFX/FXCommon_CloudToonShading
graph {nodes [
  n1 = Transform(None=NamedRerouteUsage())
  n2 = CameraVectorWS()
  n3 = Saturate(None=OneMinus(None=DotProduct(A=@n1, B=@n2)))
  n4 = ScalarParameter() {defaultValue=5, group=39.RimLight, parameterName=CustomShading_RimLight_Power}
  n5 = Multiply(A=Saturate(None=Power(Base=@n3, Exp=@n4)), B=ComponentMask(None=NamedRerouteUsage()) {r=true})
  n6 = NamedRerouteUsage()
  n7 = Transform(None=MaterialFunctionCall(Normal=NamedRerouteUsage(), Flatness=ScalarParameter() {group=30.LightingAnimeShadow, parameterName=CustomShading_Shadow_NormalFlat}) {materialFunction=&FlattenNormal}.Result)
  n8 = Multiply(
   A=Add(A=DotProduct(A=@n6, B=@n7), B=ScalarParameter() {group=30.LightingAnimeShadow, parameterName=CustomShading_Shadow_Offset})
   B=ScalarParameter() {defaultValue=1, group=30.LightingAnimeShadow, parameterName=CustomShading_Shadow_Mul}
  )
  n9 = Saturate(None=Power(Base=Saturate(None=@n8), Exp=ScalarParameter() {defaultValue=1, group=30.LightingAnimeShadow, parameterName=CustomShading_Shadow_Power}))
  n10 = Constant()
  n11 = Constant() {r=1}
  n12 = ScalarParameter() {parameterName=CustomShading_Shadow_RemapLow}
  n13 = Constant() {r=1}
  n14 = Multiply(
   A=Saturate(None=MaterialFunctionCall(Input=@n9, Input Low=@n10, Input High=@n11, Target Low=@n12, Target High=@n13) {materialFunction=&RemapValueRange}.Result)
   B=ComponentMask(None=NamedRerouteUsage()) {g=true}
  )
  n15 = LinearInterpolate(
   A=LinearInterpolate(
    A=VectorParameter() {parameterName=CustomShading_Color_Lit, defaultValue=<rgba 0 0 0 1>}.RGB
    B=VectorParameter() {parameterName=CustomShading_Color_LitRim, defaultValue=<rgba 0 0 0 1>}.RGB
    Alpha=Saturate(None=@n5)
   )
   B=VectorParameter() {parameterName=CustomShading_Color_Shadow, defaultValue=<rgba 0 0 0 1>}.RGB
   Alpha=OneMinus(None=Saturate(None=@n14))
  )
  n16 = NamedRerouteUsage()
  n17 = ScalarParameter() {group=30.LightingAnimeShadow, parameterName=CustomShading_SoftShadow_NormalFlat}
  n18 = DotProduct(
   A=NamedRerouteUsage()
   B=Transform(None=MaterialFunctionCall(Normal=@n16, Flatness=@n17) {materialFunction=&FlattenNormal}.Result)
  )
  n19 = ScalarParameter() {group=30.LightingAnimeShadow, parameterName=CustomShading_SoftShadow_Offset}
  n20 = Saturate(None=Multiply(A=Add(A=@n18, B=@n19), B=ScalarParameter() {defaultValue=1, group=30.LightingAnimeShadow, parameterName=CustomShading_SoftShadow_Mul}))
  n21 = ScalarParameter() {defaultValue=1, group=30.LightingAnimeShadow, parameterName=CustomShading_SoftShadow_Power}
  n22 = MaterialFunctionCall(
   Input=Saturate(None=Power(Base=@n20, Exp=@n21))
   Input Low=Constant()
   Input High=Constant() {r=1}
   Target Low=ScalarParameter() {parameterName=CustomShading_SoftShadow_RemapLow}
   Target High=Constant() {r=1}
  ) {materialFunction=&RemapValueRange}
  n23 = NamedRerouteUsage()
  n24 = Saturate(None=Multiply(A=Saturate(None=@n22.Result), B=ComponentMask(None=@n23) {b=true}))
  n25 = NamedRerouteUsage()
  n26 = Transform(None=MaterialFunctionCall(Normal=NamedRerouteUsage(), Flatness=ScalarParameter() {group=30.LightingAnimeShadow, parameterName=CustomShading_FakeGI_NormalFlat}) {materialFunction=&FlattenNormal}.Result)
  n27 = Multiply(
   A=Add(A=DotProduct(A=@n25, B=@n26), B=ScalarParameter() {group=30.LightingAnimeShadow, parameterName=CustomShading_FakeGI_Offset})
   B=ScalarParameter() {defaultValue=1, group=30.LightingAnimeShadow, parameterName=CustomShading_FakeGI_Mul}
  )
  n28 = Power(
   Base=OneMinus(None=Saturate(None=@n27))
   Exp=ScalarParameter() {defaultValue=1, group=30.LightingAnimeShadow, parameterName=CustomShading_FakeGI_Power}
  )
  n29 = MaterialFunctionCall(
   Input=Saturate(None=OneMinus(None=@n28))
   Input Low=Constant()
   Input High=Constant() {r=1}
   Target Low=ScalarParameter() {parameterName=CustomShading_FakeGI_RemapLow}
   Target High=Constant() {r=1}
  ) {materialFunction=&RemapValueRange}
  n30 = NamedRerouteUsage()
  n31 = Saturate(None=Multiply(A=Saturate(None=@n29.Result), B=ComponentMask(None=@n30) {a=true}))
  FunctionOutput(None=LinearInterpolate(
    A=Multiply(A=@n15, B=@n24)
    B=VectorParameter() {parameterName=CustomShading_Color_FakeGI, defaultValue=<rgba 0 0 0 1>}.RGB
    Alpha=OneMinus(None=@n31)
   )) {outputName=Emmissive}
  NamedRerouteDeclaration(None=FunctionInput(Preview=Constant3Vector() {constant=<rgba 0 0 1 0>}) {bUsePreviewValueAsDefault=true, inputName=PixelNormalTS}) {name=PixelNormalTSIn, nodeColor=<rgba 0 1 0.6719 1>}
  NamedRerouteDeclaration(None=FunctionInput(Preview=SkyAtmosphereLightDirection()) {bUsePreviewValueAsDefault=true, inputName=LightDirIn}) {name=LightDirIn, nodeColor=<rgba 0 0.4375 1 1>}
  n32 = FunctionInput() {
   bUsePreviewValueAsDefault=true
   desc=Rimlight/Shadow/SoftShadow/FakeGI
   inputName=Masks
   inputType=FunctionInput_Vector4
   previewValue=<1 1 1 1>
  }
  NamedRerouteDeclaration(None=@n32) {name=Masks, nodeColor=<rgba 1 0.0938 0 1>}
  OneMinus(None=@n24)
 ]}
