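# Cloth-shading-model material for Niagara mesh particles, apparently the cloud effect on the Genius Blob glider.
# A toon cloud shading function takes the sky atmosphere light direction, blended toward the FELightDir parameter by a saturated ceiling of its absolute component sum.
# That sum counts R twice and omits G, so a green-only override is ignored.
# The toon emissive passes through an HSV adjustment into the shared emissive adjustments function.
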Material /BRCosmetics/Gadgets/Assets/VinderTech_GliderChute/Glider_Genius_Blob/FX/M_GeniusBlob_Cloud_SP
graph {
 properties {bUsedWithNiagaraMeshParticles=true, pixelDepthOffsetMode=PDOM_Legacy, shadingModel=MSM_Cloth, usedShadingModels=Cloth}
 nodes [
  Constant3Vector() {constant=<rgba 0.003 0.001 0.002 1>}
  Constant()
  Constant() {r=1}
  Constant() {r=0.969}
  n1 = VectorParameter() {parameterName=FELightDir, defaultValue=<rgba 0 0 0 1>}
  n2 = Abs(None=Add(A=Add(A=@n1.R, B=@n1.R), B=@n1.B))
  n3 = LinearInterpolate(A=SkyAtmosphereLightDirection(), B=@n1.RGB, Alpha=Saturate(None=Ceil(None=@n2)))
  MaterialFunctionCall(Emissive=MaterialFunctionCall!(
    HSV=VectorParameter() {parameterName=HSV, defaultValue=<rgba 0 1 1 1>}.RGB
    Color=MaterialFunctionCall!(LightDirIn=@n3, Masks, PixelNormalTS) {materialFunction=/BRCosmetics/Gadgets/Assets/VinderTech_GliderChute/Glider_Genius_Blob/VFX/FXCommon_CloudToonShading}.Emmissive
   ) {materialFunction=/Game/Effects/Niagara/MasterMaterials/Functions/Common/FXCommon_HSV}.Result) {materialFunction=&MF_FX_EmissiveAdjustments}
 ]
}
