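# Cloud material function for the VinderTech Genius Blob glider VFX that outputs world position offset, a tangent normal, raw heights and masks.
# Panning main and detail height textures appear to displace vertices along their normals, with one term scaled per local axis.
# Sphere masks around a flat center and two foot positions provide flattening regions.
# An optional projection-direction seam mask blends the normal map with the vertex normal to hide UV seams.
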MaterialFunction /BRCosmetics/Gadgets/Assets/VinderTech_GliderChute/Glider_Genius_Blob/VFX/FXCommon_CloudWPO
graph {nodes [
  n1 = Multiply(A=Multiply(A=NamedRerouteUsage(), B=VertexNormalWS()), B=NamedRerouteUsage())
  n2 = Normalize(VectorInput=Transform(None=Constant3Vector() {constant=<rgba 1 0 0 0>}) {transformSourceType=TRANSFORMSOURCE_Local})
  n3 = MaterialFunctionCall(Float3=VectorParameter() {parameterName=Cloud_Main_WPOScale3Axis, defaultValue=<rgba 1 1 1 1>}.RGB) {materialFunction=&BreakOutFloat3Components}
  n4 = Multiply(A=Multiply(A=DotProduct(A=@n1, B=@n2), B=@n3.R), B=@n2)
  n5 = Normalize(VectorInput=Transform(None=Constant3Vector() {constant=<rgba 0 1 0 0>}) {transformSourceType=TRANSFORMSOURCE_Local})
  n6 = Multiply(A=Multiply(A=DotProduct(A=@n1, B=@n5), B=@n3.G), B=@n5)
  n7 = Normalize(VectorInput=Transform(None=Constant3Vector() {constant=<rgba 0 0 1 0>}) {transformSourceType=TRANSFORMSOURCE_Local})
  n8 = Multiply(A=DotProduct(A=@n1, B=@n7), B=@n3.B)
  n9 = NamedRerouteUsage()
  n10 = VertexNormalWS()
  n11 = Add(A=Add(A=Add(A=@n4, B=@n6), B=Multiply(A=@n8, B=@n7)), B=Multiply(A=Multiply(A=@n9, B=@n10), B=NamedRerouteUsage()))
  n12 = Max(A=Max(A=NamedRerouteUsage(), B=NamedRerouteUsage()), B=NamedRerouteUsage())
  FunctionOutput(None=Multiply(A=LinearInterpolate(A=@n11, Alpha=@n12) {constB=0}, B=NamedRerouteUsage())) {outputName=WPO}
  n13 = VertexNormalWS()
  n14 = NamedRerouteUsage()
  n15 = Normalize(VectorInput=Transform(None=@n13) {transformSourceType=TRANSFORMSOURCE_World, transformType=TRANSFORM_Local})
  n16 = VectorParameter() {parameterName=Cloud_Normal_UVProjectDirection, defaultValue=<rgba 0 0 1 1>}
  n17 = VectorParameter() {parameterName=Cloud_Normal_SeamMaskAdjust, channelNames {a="", b="", g=Mul, r=Power}, defaultValue=<rgba 2 3 0 1>}
  n18 = Power(Base=Abs(None=DotProduct(A=@n15, B=@n16.RGB)), Exp=@n17.Power)
  n19 = LinearInterpolate(
   A=Normalize(VectorInput=Transform(None=@n13) {transformSourceType=TRANSFORMSOURCE_World, transformType=TRANSFORM_Tangent})
   B=@n14
   Alpha=Saturate(None=Multiply(A=@n18, B=@n17.Mul))
  )
  n20 = NamedRerouteUsage()
  n21 = NamedRerouteUsage()
  FunctionOutput(None=LinearInterpolate(
    A=StaticSwitchParameter(True=@n19, False=@n14) {defaultValue=true, parameterName=Cloud_Normal_UseTangentSeamMask}
    B=Constant3Vector() {constant=<rgba 0 0 1 0>}
    Alpha=Multiply(A=@n20, B=@n21)
   ) {constB=0}) {outputName=NormalTS}
  n22 = TextureObjectParameter() {
   group=Cloud_Main_HeightTex
   parameterName=Cloud_Main_Height
   samplerType=SAMPLERTYPE_LinearGrayscale
   sortPriority=0
   texture=/Game/Gadgets/Assets/VinderTech_GliderChute/Glider_Thrive_Female/FX/Texture/T_CloudBubble
  }
  n23 = VectorParameter() {group=Cloud_Main_HeightTex, parameterName=Cloud_Main_UVOffset, channelNames {a="", b="", g=V, r=U}, defaultValue=<rgba 1 1 0 0>}
  n24 = VectorParameter() {
   group=Cloud_Main_HeightTex
   parameterName=Cloud_Main_UVControl
   channelNames {a=VPan, b=UPan, g=VScale, r=UScale}
   defaultValue=<rgba 1 1 0 0>
  }
  n25 = NamedRerouteUsage()
  n26 = MaterialFunctionCall(
   Texture=@n22
   AdditionalUVOffset=@n23
   UVControl_RGB=@n24
   UVControl_A=@n24
   "[Override]TimeInput"=NamedRerouteUsage()
   "[Override]UVInput"=@n25
  ) {materialFunction=&FXCommon_TextureSample}
  FunctionOutput(None=@n26.R) {outputName=OriginMainHeight}
  n27 = TextureObjectParameter() {
   group=Cloud_Main_HeightTex
   parameterName=Cloud_Main_Normal
   samplerType=SAMPLERTYPE_Normal
   sortPriority=0
   texture=&/T_CloudBubble_Normal
  }
  n28 = MaterialFunctionCall(
   Texture=@n27
   AdditionalUVOffset=@n23
   UVControl_RGB=@n24
   UVControl_A=@n24
   "[Override]TimeInput"=NamedRerouteUsage()
   "[Override]UVInput"=@n25
  ) {materialFunction=&FXCommon_TextureSample}
  FunctionOutput(None=@n28.RGB) {outputName=OriginMainNormal}
  n29 = TextureObjectParameter() {
   group=Cloud_Main_HeightTex
   parameterName=Cloud_Detail_Height
   samplerType=SAMPLERTYPE_LinearGrayscale
   sortPriority=0
   texture=&T_CloudBubble
  }
  n30 = VectorParameter() {group=Cloud_Main_HeightTex, parameterName=Cloud_Detail_UVOffset, channelNames {a="", b="", g=V, r=U}, defaultValue=<rgba 1 1 0 0>}
  n31 = VectorParameter() {
   group=Cloud_Main_HeightTex
   parameterName=Cloud_Detail_UVControl
   channelNames {a=VPan, b=UPan, g=VScale, r=UScale}
   defaultValue=<rgba 1 1 0 0>
  }
  n32 = MaterialFunctionCall(
   Texture=@n29
   AdditionalUVOffset=@n30
   UVControl_RGB=@n31
   UVControl_A=@n31
   "[Override]TimeInput"=NamedRerouteUsage()
   "[Override]UVInput"=NamedRerouteUsage()
  ) {materialFunction=&FXCommon_TextureSample}
  FunctionOutput(None=@n32.R) {outputName=OriginDetailHeight}
  n33 = SphereMask(
   A=WorldPosition().XYZ
   B=NamedRerouteUsage()
   Radius=ScalarParameter() {parameterName=Cloud_FlatRadius}
   Hardness=ScalarParameter() {parameterName=Cloud_FlatHardness}
  )
  FunctionOutput(None=@n33) {outputName=FlatMask}
  n34 = SphereMask(
   A=WorldPosition().XYZ
   B=NamedRerouteUsage()
   Radius=ScalarParameter() {parameterName=Cloud_FootLeft_Radius}
   Hardness=ScalarParameter() {parameterName=Cloud_FootLeft_Hardness}
  )
  FunctionOutput(None=@n34) {outputName=FootLeftMask}
  n35 = SphereMask(
   A=WorldPosition().XYZ
   B=NamedRerouteUsage()
   Radius=ScalarParameter() {parameterName=Cloud_FootRight_Radius}
   Hardness=ScalarParameter() {parameterName=Cloud_FootRight_Hardness}
  )
  FunctionOutput(None=@n35) {outputName=FootRightMask}
  NamedRerouteDeclaration(None=FunctionInput(Preview=Time() {bOverride_Period=true, period=9999}) {bUsePreviewValueAsDefault=true, inputName=TimeIn, inputType=FunctionInput_Scalar}) {name=TimeIn, nodeColor=<rgba 0 1 0.9297 1>}
  NamedRerouteDeclaration(None=FunctionInput(Preview=TextureCoordinate()) {bUsePreviewValueAsDefault=true, inputName=CloudUV_Main, inputType=FunctionInput_Vector2}) {name=MainUV, nodeColor=<rgba 1 0.2812 0 1>}
  NamedRerouteDeclaration(None=@n33) {name=FlatMask, nodeColor=<rgba 0.2422 1 0 1>}
  n36 = FunctionInput() {bUsePreviewValueAsDefault=true, inputName=Mask, inputType=FunctionInput_Scalar, previewValue=<1 0 0 1>}
  NamedRerouteDeclaration(None=@n36) {name=Mask, nodeColor=<rgba 1 0 0.1173 1>}
  NamedRerouteDeclaration(None=FunctionInput() {bUsePreviewValueAsDefault=true, inputName=FootHole_Left_Position}) {name=FootHole_Left_Position, nodeColor=<rgba 0 1 0.0156 1>}
  NamedRerouteDeclaration(None=FunctionInput() {bUsePreviewValueAsDefault=true, inputName=FootHole_Right_Position}) {name=FootHole_Right_Position, nodeColor=<rgba 0 0.0859 1 1>}
  NamedRerouteDeclaration(None=@n34) {name=FootHole_LeftMask, nodeColor=<rgba 1 0 0.5625 1>}
  NamedRerouteDeclaration(None=@n35) {name=FootHole_RightMask, nodeColor=<rgba 1 0 0.5625 1>}
  n37 = FunctionInput() {bUsePreviewValueAsDefault=true, inputName=CloudScale, inputType=FunctionInput_Scalar, previewValue=<1 0 0 1>}
  NamedRerouteDeclaration(None=@n37) {name=CloudScale, nodeColor=<rgba 0 0.5078 1 1>}
  NamedRerouteDeclaration(None=@n28.RGB) {name=Cloud_Main_Normal, nodeColor=<rgba 1 0.5391 0 1>}
  n38 = Power(Base=@n26.R, Exp=ScalarParameter() {defaultValue=1, parameterName=Cloud_Main_HeightPower})
  n39 = ScalarParameter() {defaultValue=-0.5, parameterName=Cloud_Main_HeightOffset}
  NamedRerouteDeclaration(None=Multiply(A=Add(A=@n38, B=@n39), B=ScalarParameter() {defaultValue=100, parameterName=Cloud_Main_HeightStrength})) {name=Cloud_Main_Height, nodeColor=<rgba 1 0 0.1406 1>}
  NamedRerouteDeclaration(None=FunctionInput(Preview=TextureCoordinate() {coordinateIndex=1}) {bUsePreviewValueAsDefault=true, inputName=CloudUV_Detail, inputType=FunctionInput_Vector2}) {name=DetailUV, nodeColor=<rgba 1 0.2812 0 1>}
  n40 = Power(Base=@n32.R, Exp=ScalarParameter() {defaultValue=1, parameterName=Cloud_Detail_HeightPower})
  n41 = ScalarParameter() {parameterName=Cloud_Detail_HeightOffset}
  NamedRerouteDeclaration(None=Multiply(A=Add(A=@n40, B=@n41), B=ScalarParameter() {defaultValue=100, parameterName=Cloud_Detail_HeightStrength})) {name=Cloud_Detail_Height, nodeColor=<rgba 1 0 0.1406 1>}
  NamedRerouteDeclaration(None=FunctionInput(Preview=ActorPositionWS()) {bUsePreviewValueAsDefault=true, inputName=FlatMaskCenterPos}) {name=FlatMaskCenterPos, nodeColor=<rgba 0.6406 1 0 1>}
 ]}
