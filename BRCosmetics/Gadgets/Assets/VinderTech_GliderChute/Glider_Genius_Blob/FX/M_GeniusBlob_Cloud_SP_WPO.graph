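# Cloth-shaded material for Niagara mesh-particle clouds on the Genius Blob glider.
# A shared cloud WPO function takes scale and age from dynamic parameters plus scaled time, and its normal feeds a toon shading function.
# Toon lighting uses the sky atmosphere direction unless FELightDir's red/blue sum overrides it (green is ignored), then passes through HSV and emissive adjustments.
# Particle alpha drives a temporal dither and optional opacity.
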Material /BRCosmetics/Gadgets/Assets/VinderTech_GliderChute/Glider_Genius_Blob/FX/M_GeniusBlob_Cloud_SP_WPO
graph {
 properties {bUsedWithNiagaraMeshParticles=true, pixelDepthOffsetMode=PDOM_Legacy, shadingModel=MSM_Cloth, usedShadingModels=Cloth}
 nodes [
  Constant3Vector() {constant=<rgba 0.003 0.001 0.002 1>}
  Constant()
  Constant() {r=0.969}
  Constant()
  n1 = Constant() {r=1}
  n2 = ParticleColor()
  LinearInterpolate(
   A=LinearInterpolate(A=@n1, B=@n2.A, Alpha=ScalarParameter() {parameterName=Use Particle Alpha For Opacity})
   B=@n1
   Alpha=ScalarParameter() {parameterName=Is Masked}
  )
  n3 = VectorParameter() {parameterName=FELightDir, defaultValue=<rgba 0 0 0 1>}
  n4 = Abs(None=Add(A=Add(A=@n3.R, B=@n3.R), B=@n3.B))
  n5 = LinearInterpolate(A=SkyAtmosphereLightDirection(), B=@n3.RGB, Alpha=Saturate(None=Ceil(None=@n4)))
  n6 = DynamicParameter() {paramNames [ScaleX,Age,Param3,Param4]}
  n7 = Time()
  n8 = ScalarParameter() {defaultValue=1, parameterName=WPO Speed}
  n9 = MaterialFunctionCall!(
   CloudUV_Main
   CloudScale=@n6.ScaleX
   CloudUV_Detail=TextureCoordinate() {coordinateIndex=1}
   FlatMaskCenterPos
   Mask
   FootHole_Left_Position
   FootHole_Right_Position
   TimeIn=Add(A=Multiply(A=@n7, B=@n8), B=@n6.Age)
  ) {materialFunction=/BRCosmetics/Gadgets/Assets/VinderTech_GliderChute/Glider_Genius_Blob/VFX/FXCommon_CloudWPO}
  MaterialFunctionCall(Emissive=MaterialFunctionCall!(
    HSV=VectorParameter() {parameterName=HSV, defaultValue=<rgba 0 1 1 1>}.RGB
    Color=MaterialFunctionCall!(LightDirIn=@n5, Masks, PixelNormalTS=@n9.NormalTS) {materialFunction=/BRCosmetics/Gadgets/Assets/VinderTech_GliderChute/Glider_Genius_Blob/VFX/FXCommon_CloudToonShading}.Emmissive
   ) {materialFunction=/Game/Effects/Niagara/MasterMaterials/Functions/Common/FXCommon_HSV}.Result) {materialFunction=&MF_FX_EmissiveAdjustments}
  MaterialFunctionCall(Normal=@n9.NormalTS, Flatness=ScalarParameter() {defaultValue=-1, parameterName=F}) {materialFunction=&FlattenNormal}
  MaterialFunctionCall(Alpha Threshold=@n2.A) {materialFunction=&DitherTemporalAA}
 ]
}
