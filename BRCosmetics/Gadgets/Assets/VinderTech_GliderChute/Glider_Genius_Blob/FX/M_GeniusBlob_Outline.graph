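# Translucent unlit Niagara mesh-particle material that draws screen-space outlines for the Genius Blob glider.
# Four-tap scene depth and world-normal samples form inner, edge and normal lines, with normal lines tinted by sky light direction.
# Line width follows a particle LineScale parameter, resolution, FOV and distance.
# Opacity is masked by a GBuffer roughness key (opacity on low quality) and depth threshold, then distance-faded; forward and mobile drop normal and edge lines.
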Material /BRCosmetics/Gadgets/Assets/VinderTech_GliderChute/Glider_Genius_Blob/FX/M_GeniusBlob_Outline
graph {
 properties {
  allowTranslucentCustomDepthWrites=true
  bEnableResponsiveAA=true
  bUsedWithNiagaraMeshParticles=true
  blendMode=BLEND_Translucent
  pixelDepthOffsetMode=PDOM_Legacy
  shadingModel=MSM_Unlit
  usedShadingModels=Unlit
 }
 nodes [
  Constant()
  n1 = Constant()
  If(
   A=ComponentMask(None=SceneTexture() {sceneTextureId=PPI_SubsurfaceColor}.Color) {b=true, g=true, r=true}
   B=Constant3Vector() {constant=<rgba 0.003 0.001 0.002 1>}
   "A > B"=@n1
   "A == B"=Constant() {r=1}
   "A < B"=@n1
  ) {equalsThreshold=0.05}
  n2 = VectorParameter() {group=InLine, parameterName=InLineColor, defaultValue=<rgba 0.05 0.0199 0 1>}
  n3 = VectorParameter() {group=NormalLine, parameterName=NormalLineColor_D, defaultValue=<rgba 0.1146 0.0477 0 1>}
  n4 = MaterialFunctionCall(Hue Shift Percentage=ScalarParameter() {group=NormalLine, parameterName=Lighting_NormalLine_HueShift}, Texture=@n3.RGB) {materialFunction=&HueShift}
  n5 = ScalarParameter() {group=NormalLine, parameterName=Lighting_NormalLine_Desaturation}
  n6 = Subtract(A=WorldPosition().XYZ, B=ObjectPositionWS())
  n7 = Add(
   A=DotProduct(A=SkyAtmosphereLightDirection(), B=Normalize(VectorInput=@n6))
   B=ScalarParameter() {group=NormalLineLighting, parameterName=Lighting_NormalLine_NDS_Offset}
  )
  n8 = ScalarParameter() {defaultValue=1, group=NormalLineLighting, parameterName=Lighting_NormalLine_NDS_Mul}
  n9 = Power(
   Base=Saturate(None=Multiply(A=@n7, B=@n8))
   Exp=ScalarParameter() {defaultValue=1, group=NormalLineLighting, parameterName=Lighting_NormalLine_NDS_Power}
  )
  n10 = VectorParameter() {group=NormalLineLighting, parameterName=Lighting_NormalLine_NDS_Remap, defaultValue=<rgba 0 1 0 1>}
  n11 = Saturate(None=MaterialFunctionCall(Input=Saturate(None=@n9), Input Low=Constant(), Input High=Constant() {r=1}, Target Low=@n10.R, Target High=@n10.R) {materialFunction=&RemapValueRange}.Result)
  n12 = Multiply(A=LinearInterpolate(A=Desaturation(None=@n4.Result, Fraction=@n5), B=@n3.RGB, Alpha=@n11), B=@n11)
  n13 = ComponentMask(None=ScreenPosition().ViewportUV) {g=true, r=true}
  n14 = Divide(A=ViewSize(), B=Constant() {r=320})
  n15 = Constant() {r=2}
  n16 = Clamp(None=Ceil(None=Subtract(A=@n14, B=@n15) {constB=0})) {maxDefault=2, minDefault=1}
  n17 = SceneTexelSize()
  n18 = DynamicParameter() {paramNames [Param1,Param2,LineScale,Param4]}
  n19 = Multiply(A=ScalarParameter() {defaultValue=16, group=NormalLine, parameterName=NormalLineThickness}, B=@n18.LineScale)
  n20 = ViewSize()
  n21 = Divide(A=ComponentMask(None=@n20) {r=true}, B=ComponentMask(None=@n20) {g=true})
  n22 = MaterialFunctionCall() {materialFunction=/Engine/Functions/Engine_MaterialFunctions02/Utility/FOV}
  n23 = Distance(A=ObjectPositionWS(), B=CameraPositionWS())
  n24 = ScalarParameter() {defaultValue=0.05, parameterName=ThicknessFalloffSpeed}
  n25 = Power(Base=Divide(A=Divide(A=@n21, B=@n22."FOV "), B=Multiply(A=@n23, B=@n24) {constB=0.05})) {constExponent=0.6}
  n26 = Multiply(A=@n16, B=Multiply(A=@n17, B=Multiply(A=@n25, B=@n19))) {constA=2}
  n27 = Constant()
  n28 = AppendVector(A=ComponentMask(None=@n26) {desc=x, r=true}, B=@n27)
  n29 = Add(A=@n13, B=@n28) {desc=x+}
  n30 = SceneTexture() {sceneTextureId=PPI_WorldNormal}
  n31 = AppendVector(A=@n27, B=ComponentMask(None=@n26) {desc=y, g=true}) {desc=y+}
  n32 = Add(A=@n13, B=@n31) {desc=y+}
  n33 = Add(
   A=Distance(A=SceneTexture(UVs=@n29) {sceneTextureId=PPI_WorldNormal}.Color, B=@n30.Color)
   B=Distance(A=SceneTexture(UVs=@n32) {sceneTextureId=PPI_WorldNormal}.Color, B=@n30.Color)
  )
  n34 = Constant() {r=-1}
  n35 = Multiply(A=@n28, B=@n34)
  n36 = Distance(A=SceneTexture(UVs=Add(A=@n13, B=@n35) {desc=x-}) {sceneTextureId=PPI_WorldNormal}.Color, B=@n30.Color)
  n37 = SceneTexture(UVs=Add(A=@n13, B=Multiply(A=@n34, B=@n31)) {desc=y-}) {sceneTextureId=PPI_WorldNormal}
  n38 = Divide(A=Add(A=Add(A=@n33, B=@n36), B=Distance(A=@n37.Color, B=@n30.Color)), B=Constant() {r=4}) {constB=1}
  n39 = ScalarParameter() {defaultValue=0.25, group=NormalLine, parameterName=NSampleBias}
  n40 = OneMinus(None=Saturate(None=Subtract(A=@n38, B=@n39)))
  n41 = ScalarParameter() {defaultValue=0.3, group=NormalLine, parameterName=NormalLinePower}
  n42 = Multiply(A=OneMinus(None=Power(Base=@n40, Exp=@n41)), B=ScalarParameter() {defaultValue=12, group=NormalLine, parameterName=NormalLineScale})
  n43 = VectorParameter() {parameterName=Fade, defaultValue=<rgba 4096 200 500 100>}
  n44 = Divide(A=Add(A=@n23, B=@n43.B), B=@n43.A) {constB=4096}
  n45 = Multiply(
   A=OneMinus(None=Saturate(None=@n44))
   B=Power(Base=Saturate(None=@n42), Exp=ScalarParameter() {defaultValue=50, group=NormalLine, parameterName=NormalLinePower2})
  )
  n46 = Constant()
  n47 = ShadingPathSwitch(Default=@n45, Deferred=@n45, Forward=@n46, Mobile=@n46)
  n48 = ComponentMask(None=SceneDepth().R) {r=true}
  n49 = Subtract(A=@n48, B=ScalarParameter() {defaultValue=-1.1, group=InLine, parameterName=SampleBias})
  n50 = ComponentMask(None=ScreenPosition().ViewportUV) {g=true, r=true}
  n51 = Multiply(A=Multiply(A=@n18.LineScale, B=ScalarParameter() {defaultValue=6, group=InLine, parameterName=LineThickness}), B=@n25)
  n52 = Multiply(A=@n16, B=Multiply(A=@n17, B=Clamp(None=@n51) {clampMode=CMODE_ClampMin})) {constA=2}
  n53 = Constant()
  n54 = AppendVector(A=ComponentMask(None=@n52) {desc=x, r=true}, B=@n53)
  n55 = Add(A=@n50, B=@n54) {desc=x+}
  n56 = ComponentMask(None=SceneDepth(UVs=@n55).R) {r=true}
  n57 = AppendVector(A=@n53, B=ComponentMask(None=@n52) {desc=y, g=true}) {desc=y+}
  n58 = Add(A=@n50, B=@n57) {desc=y+}
  n59 = ComponentMask(None=SceneDepth(UVs=@n58).R) {r=true}
  n60 = Constant() {r=-1}
  n61 = Add(A=@n50, B=Multiply(A=@n54, B=@n60)) {desc=x-}
  n62 = SceneDepth(UVs=@n61)
  n63 = Add(A=Add(A=Subtract(A=@n49, B=@n56), B=Subtract(A=@n49, B=@n59)), B=Subtract(A=@n49, B=ComponentMask(None=@n62.R) {r=true}))
  n64 = Add(A=@n50, B=Multiply(A=@n60, B=@n57)) {desc=y-}
  n65 = Subtract(A=@n49, B=ComponentMask(None=SceneDepth(UVs=@n64).R) {r=true})
  n66 = Saturate(None=Divide(A=Add(A=@n63, B=@n65), B=ScalarParameter() {defaultValue=-3, group=InLine, parameterName=Divisor}) {constB=1})
  n67 = ScalarParameter() {defaultValue=1, group=InLine, parameterName=ScaleLine}
  n68 = Saturate(None=Power(Base=Multiply(A=@n66, B=@n67), Exp=ScalarParameter() {defaultValue=2, group=InLine, parameterName=Powerline}))
  n69 = Saturate(None=Subtract(A=@n47, B=@n68))
  n70 = Subtract(A=@n48, B=ScalarParameter() {defaultValue=-3, group=EdgeLine, parameterName=EdgeLine_SampleBias})
  n71 = ComponentMask(None=ScreenPosition().ViewportUV) {g=true, r=true}
  n72 = Multiply(A=Multiply(A=ScalarParameter() {defaultValue=17, group=EdgeLine, parameterName=EdgeLineLineThickness}, B=@n18.LineScale), B=@n25)
  n73 = Multiply(A=@n16, B=Multiply(A=@n17, B=Clamp(None=@n72) {clampMode=CMODE_ClampMin}))
  n74 = Constant()
  n75 = AppendVector(A=ComponentMask(None=@n73) {desc=x, r=true}, B=@n74)
  n76 = Add(A=@n71, B=@n75) {desc=x+}
  n77 = ComponentMask(None=SceneDepth(UVs=@n76).R) {r=true}
  n78 = AppendVector(A=@n74, B=ComponentMask(None=@n73) {desc=y, g=true}) {desc=y+}
  n79 = Add(A=@n71, B=@n78) {desc=y+}
  n80 = ComponentMask(None=SceneDepth(UVs=@n79).R) {r=true}
  n81 = Constant() {r=-1}
  n82 = Add(A=@n71, B=Multiply(A=@n75, B=@n81)) {desc=x-}
  n83 = SceneDepth(UVs=@n82)
  n84 = Add(A=Add(A=Subtract(A=@n70, B=@n77), B=Subtract(A=@n70, B=@n80)), B=Subtract(A=@n70, B=ComponentMask(None=@n83.R) {r=true}))
  n85 = Add(A=@n71, B=Multiply(A=@n81, B=@n78)) {desc=y-}
  n86 = Subtract(A=@n70, B=ComponentMask(None=SceneDepth(UVs=@n85).R) {r=true})
  n87 = Saturate(None=Divide(A=Add(A=@n84, B=@n86), B=ScalarParameter() {defaultValue=-700, group=EdgeLine, parameterName=EdgeLine_Divisor}) {constB=1})
  n88 = ScalarParameter() {defaultValue=3, group=EdgeLine, parameterName=EdgeLine_ScaleLine}
  n89 = Saturate(None=Power(Base=Multiply(A=@n87, B=@n88), Exp=ScalarParameter() {defaultValue=5, group=EdgeLine, parameterName=EdgeLine_PowerLine}))
  n90 = Constant()
  n91 = ShadingPathSwitch(Default=@n89, Deferred=@n89, Forward=@n90, Mobile=@n90)
  n92 = ScalarParameter() {parameterName=Debug}
  n93 = LinearInterpolate(
   A=LinearInterpolate(
    A=LinearInterpolate(A=@n2.RGB, B=@n12, Alpha=@n69)
    B=VectorParameter() {parameterName=LineColor, defaultValue=<rgba 0 0 0 1>}.RGB
    Alpha=@n91
   )
   B=@n91
   Alpha=@n92
  )
  MaterialFunctionCall(Emissive=MaterialFunctionCall!(
    HSV=VectorParameter() {parameterName=HSV, defaultValue=<rgba 0 1 1 1>}.RGB
    Color=ShadingPathSwitch(Default=@n93, Deferred=@n93, Forward=@n93, Mobile=@n93)
   ) {materialFunction=/Game/Effects/Niagara/MasterMaterials/Functions/Common/FXCommon_HSV}.Result) {materialFunction=&MF_FX_EmissiveAdjustments}
  n94 = Constant()
  n95 = If(
   A=MaterialFunctionCall(Float3=SceneTexture() {sceneTextureId=PPI_Roughness}.Color) {materialFunction=&BreakOutFloat3Components}.R
   "A > B"=@n94
   "A == B"=Constant() {r=1}
   "A < B"=@n94
  ) {constB=0.969, equalsThreshold=0.003}
  n96 = QualitySwitch(
   Default=@n95
   Low=Floor(None=SceneTexture() {sceneTextureId=PPI_Opacity}.Color) {desc="for effect quality low (SSColor is disabled)"}
   High=@n95
   Medium=@n95
   Epic=@n95
  )
  n97 = Constant() {r=1}
  n98 = Add(A=PixelDepth().R, B=ScalarParameter() {defaultValue=400, parameterName=LineDistanceThresold})
  n99 = Constant()
  n100 = Constant() {r=1}
  n101 = Max(A=@n91, B=@n68)
  n102 = LinearInterpolate(A=Multiply(A=If(A=@n48, B=@n98, "A > B"=@n99, "A == B"=@n100, "A < B"=@n100), B=Max(A=@n101, B=@n47)), Alpha=@n92)
  n103 = Divide(A=Add(A=@n23, B=@n43.R), B=@n43.G) {constB=4096}
  Multiply(
   A=OneMinus(None=Saturate(None=@n103))
   B=Multiply(
    A=ShadingPathSwitch(Default=@n96, Deferred=@n96, Forward=@n97, Mobile=@n97)
    B=ShadingPathSwitch(Default=@n102, Deferred=@n102, Forward=@n102, Mobile=@n102)
   )
  )
  SceneTexture(UVs=@n76) {sceneTextureId=PPI_CustomDepth}
  SceneTexture(UVs=@n79) {sceneTextureId=PPI_CustomDepth}
  SceneTexture(UVs=@n82) {sceneTextureId=PPI_CustomDepth}
  SceneTexture(UVs=@n85) {sceneTextureId=PPI_CustomDepth}
  SceneTexture() {sceneTextureId=PPI_CustomDepth}
  SceneTexture(UVs=@n55) {sceneTextureId=PPI_CustomDepth}
  SceneTexture(UVs=@n58) {sceneTextureId=PPI_CustomDepth}
  SceneTexture(UVs=@n61) {sceneTextureId=PPI_CustomDepth}
  SceneTexture(UVs=@n64) {sceneTextureId=PPI_CustomDepth}
 ]
}
