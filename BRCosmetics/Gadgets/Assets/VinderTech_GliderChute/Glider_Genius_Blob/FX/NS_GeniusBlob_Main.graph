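# Niagara system for the Genius Blob glider cosmetic: a toon-shaded cloud mesh with speed lines.
# The mesh pops in with sine wobble and curve-driven jiggle; User.Deployed offsets curve time by 100 to skip these intros.
# Dynamic material parameters pass scale, timer, two foot positions and time-of-day values to the shader.
# GPU sprite streaks spawn from a torus at a timer-scaled rate, and User.IsFrontEnd switches lobby colors, rates and offsets.
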NiagaraSystem /BRCosmetics/Gadgets/Assets/VinderTech_GliderChute/Glider_Genius_Blob/FX/NS_GeniusBlob_Main
dependencies {
 usedDataInterfaces [&NiagaraDataInterfaceCurve]
 usedDynamicInputs [&SetBoolByFloatComparison]
 usedModules [&SystemState, &InterpolateOverTime]
}
emitters {
 NE_StaminaGliderCloud {
  emitterUpdateScript [
   &EmitterState(
    Inactive Response=::NewEnumerator1
    Life Cycle Mode=::NewEnumerator1
    Loop Behavior=::NewEnumerator1
    Loop Delay=1.1
    Loop Duration Mode=::NewEnumerator1
    UseLoopDelay=false
   )
   &SpawnBurst_Instantaneous(Use Spawn Probability=false, UseLoopCountLimit=false)
   /SetVariables(Emitter.FELight:LinearColor=&MakeCustomLinearColorFromBool(
     Bool=$User.IsFrontEnd
     True Color=&MakeLinearColorFromVectorAndFloat(
      "Vector (RGB)"=&NormalizeVector(Vector To Normalize=<0 1 1>, Safe Normalize=true, Safe Fallback Threshold=0, Safe Fallback Vector=<0 0 0.001>)
      "Float (Alpha)"=1
     )
     False Color=<rgba 0 0 0 0>
    )) {isSetParameters=true}
   /SetVariables(Emitter.TOD:Vector3f=&MakeCustomVectorFromBool(Bool=$User.IsFrontEnd, True Vector=<0 1 1>, False Vector=&MakeVector(X=0, Y=1, Z=1))) {isSetParameters=true}
   /SetVariables(Emitter.LitRim:LinearColor=&MakeCustomLinearColorFromBool(Bool=$User.IsFrontEnd, True Color=<rgba 0.25 1 1 1>, False Color=<rgba 0.2 0.8638 1 1>)) {isSetParameters=true}
   /SetVariables(Emitter.Lit:LinearColor=&MakeCustomLinearColorFromBool(Bool=$User.IsFrontEnd, True Color=<rgba 0.25 1 1 1>, False Color=<rgba 0.1 1 1 1>)) {isSetParameters=true}
   /SetVariables(Emitter.Shadow:LinearColor=&MakeCustomLinearColorFromBool(Bool=$User.IsFrontEnd, True Color=<rgba 0.12 0.5733 0.8 1>, False Color=<rgba 0.075 0.4292 0.5 1>)) {isSetParameters=true}
   /SetVariables(Emitter.GI:LinearColor=&MakeCustomLinearColorFromBool(Bool=$User.IsFrontEnd, True Color=<rgba 0.14 0.6067 0.7 1>, False Color=<rgba 0.0675 0.4181 0.45 1>)) {isSetParameters=true}
   /SetVariables(Emitter.SoftShadowStr:NiagaraFloat=&MakeCustomFloatFromBool(Bool=$User.IsFrontEnd, True Float=0.6, False Float=0.5)) {isSetParameters=true}
   /Game/Effects/Niagara/MasterMaterials/ToonShader/Template/NM_OutlineParameters(OutlineColor:LinearColor=<rgba 0 0.1866 0.224 1>)
  ]
  particleSpawnScript [
   &InitializeParticle(
    Color Mode=::NewEnumerator0
    Lifetime=5
    Mesh Scale=<0.077 0.071 0.063>
    Mesh Scale Mode=::NewEnumerator4
    Position Mode=::NewEnumerator1
    Position Offset=<0 0 -5>
    Position Offset Coordinate Space=::Local
    UsePositionOffset=true
    Write Material Random=true
   )
   &InitialMeshOrientation(
    Mesh Axis To Orient=<0 1 0>
    Mesh Orientation Mode=::NewEnumerator2
    Orientation Coordinate Space=::Local
    Orientation Vector=&NormalizeVector(Vector To Normalize=<1 0 -0.025>, Safe Normalize=true, Safe Fallback Threshold=0, Safe Fallback Vector=<0 0 0.001>)
    Use Rotation Vector=false
   )
  ]
  particleUpdateScript [
   &ParticleState(Kill Particles When Lifetime Has Elapsed=false, Let Infinitely Lived Particles Die When Emitter Deactivates=true)
   &FN_Timer(Increment Multiplier:NiagaraFloat=&MakeCustomFloatFromBool(
     Bool=$User.IsFrontEnd
     True Float=&FloatFromCurve(
      FloatCurve=DataInterface() {propertyValues {Curve {keys [<at 0 2>,<at 0.4 2>,<at 1.6 0>]}}}
      CurveIndex=&Add_Float(A=$Particles.Age, B=&Multiply_FloatByInt(Float=100, Integer=$User.Deployed))
     )
     False Float=&Multiply_Float(A=&ClampFloat(Float=$System.InterpolateOverTime.InterpolateOverTime_Float.Moving Average), B=1)
    ))
   &ScaleMeshSize(Scale Factor=&VectorFromFloat(Value=&FloatFromCurve(
      FloatCurve=DataInterface() {propertyValues {Curve {keys [
          {interpMode=RCIM_Cubic, leaveTangent=9.4486, tangentMode=RCTM_Break, time=0.3, value=0}
          {tangentMode=RCTM_User, time=0.45, value=0.9}
          <at 1 1>
         ]}}}
      CurveIndex=&Add_Float(A=$Particles.Age, B=&Multiply_FloatByInt(Float=100, Integer=$User.Deployed))
      Scale Curve=0.87
     )))
   &ScaleMeshSize(Scale Factor=&MakeVector(
     X=&Add_Float(
      A=1
      B=&Lerp_Float(
       A=&Sine(Normalized Angle=$Emitter.Age, Period=0.25, Scale=0.15, Bias=0)
       B=0
       Alpha=&FloatFromCurve(
        FloatCurve=DataInterface() {propertyValues {Curve {keys [<at 0.3 1>, {interpMode=RCIM_Cubic, leaveTangent=1.7893, tangentMode=RCTM_Break, time=0.4, value=0}, <at 2 1>]}}}
        CurveIndex=&Add_Float(A=$Particles.Age, B=&Multiply_FloatByInt(Float=100, Integer=$User.Deployed))
       )
      )
     )
     Y=&Add_Float(
      A=1
      B=&Lerp_Float(
       A=&Cosine(Normalized Angle=$Emitter.Age, Period=0.25, Scale=0.2, Bias=-0.1)
       B=0
       Alpha=&FloatFromCurve(
        FloatCurve=DataInterface() {propertyValues {Curve {keys [<at 0.3 1>, {interpMode=RCIM_Cubic, leaveTangent=1.7893, tangentMode=RCTM_Break, time=0.4, value=0}, <at 2 1>]}}}
        CurveIndex=&Add_Float(A=$Particles.Age, B=&Multiply_FloatByInt(Float=100, Integer=$User.Deployed))
       )
      )
     )
     Z=&Add_Float(
      A=1
      B=&Lerp_Float(
       A=&Sine(Normalized Angle=$Emitter.Age, Period=0.25, Scale=0.15, Bias=0)
       B=0
       Alpha=&FloatFromCurve(
        FloatCurve=DataInterface() {propertyValues {Curve {keys [<at 0.3 1>, {interpMode=RCIM_Cubic, leaveTangent=1.7893, tangentMode=RCTM_Break, time=0.4, value=0}, <at 2 1>]}}}
        CurveIndex=&Add_Float(A=$Particles.Age, B=&Multiply_FloatByInt(Float=100, Integer=$User.Deployed))
       )
      )
     )
    )) {module=ScaleMeshSize001}
   /SetVariables(Particles.New Curve for Floats=DataInterface() {propertyValues {Curve {keys [
        {interpMode=RCIM_Cubic, leaveTangent=2.6558, tangentMode=RCTM_Break, time=2.3437, value=1}
        {arriveTangent=0.3463, interpMode=RCIM_Cubic, leaveTangent=0.1262, tangentMode=RCTM_Break, time=2.3993, value=1.1353}
        {arriveTangent=-0.036, interpMode=RCIM_Cubic, leaveTangent=0.685, tangentMode=RCTM_Break, time=2.4793, value=0.8296}
        {interpMode=RCIM_Cubic, tangentMode=RCTM_Break, time=2.613, value=1.0836}
        {interpMode=RCIM_Cubic, tangentMode=RCTM_Break, time=2.71, value=0.8963}
        {interpMode=RCIM_Cubic, tangentMode=RCTM_Break, time=2.8353, value=1.0614}
        {interpMode=RCIM_Cubic, tangentMode=RCTM_Break, time=2.9233, value=0.9397}
        {interpMode=RCIM_Cubic, tangentMode=RCTM_Break, time=3.0474, value=1.0412}
        {interpMode=RCIM_Cubic, tangentMode=RCTM_Break, time=3.1713, value=0.9774}
        {interpMode=RCIM_Cubic, tangentMode=RCTM_Break, time=3.3188, value=1}
       ]}}}) {isSetParameters=true}
   /SetVariables(Particles.New Curve for Float2s=DataInterface() {propertyValues {Curve {keys [<at 0 1>]}}}) {enabled=false, isSetParameters=true}
   &ScaleMeshSize(Scale Factor=&MakeCustomVectorFromBool(
     Bool=&Boolean_OrOperation(Bool1=&SetBoolByIntComparison(A=$User.Deployed, B=1, Comparison Type=::NewEnumerator4), Bool2=$User.IsFrontEnd)
     True Vector=<1 1 1>
     False Vector=&MakeVector(
      X=&FloatFromCurve(FloatCurve=$Particles.New Curve for Floats, CurveIndex=$Particles.Age)
      Y=&Add_Float(
       A=&Multiply_Float(
        A=&Subtract_Float(A=&FloatFromCurve(FloatCurve=$Particles.New Curve for Floats, CurveIndex=$Particles.Age), B=1)
        B=-0.75
       )
       B=1
      )
      Z=&Add_Float(
       A=&Multiply_Float(
        A=&Subtract_Float(A=&FloatFromCurve(FloatCurve=$Particles.New Curve for Floats, CurveIndex=$Particles.Age), B=1)
        B=1.5
       )
       B=1
      )
     )
    )) {module=ScaleMeshSize002}
   &ScaleMeshSize(Scale Factor=&MakeVector(
     X=&Add_Float(
      A=1
      B=&Lerp_Float(
       A=&Sine(Normalized Angle=$Emitter.Age, Period=0.25, Scale=0.15, Bias=0)
       B=0
       Alpha=&FloatFromCurve(
        FloatCurve=$Particles.New Curve for Floats
        CurveIndex=&Add_Float(A=$Particles.Age, B=&Multiply_FloatByInt(Float=100, Integer=$User.Deployed))
       )
      )
     )
     Y=&Add_Float(
      A=1
      B=&Lerp_Float(
       A=&Cosine(Normalized Angle=$Emitter.Age, Period=0.25, Scale=0.2, Bias=-0.1)
       B=0
       Alpha=&FloatFromCurve(
        FloatCurve=$Particles.New Curve for Floats
        CurveIndex=&Add_Float(A=$Particles.Age, B=&Multiply_FloatByInt(Float=100, Integer=$User.Deployed))
       )
      )
     )
     Z=&Add_Float(
      A=1
      B=&Lerp_Float(
       A=&Sine(Normalized Angle=$Emitter.Age, Period=0.25, Scale=0.15, Bias=0)
       B=0
       Alpha=&FloatFromCurve(
        FloatCurve=$Particles.New Curve for Floats
        CurveIndex=&Add_Float(A=$Particles.Age, B=&Multiply_FloatByInt(Float=100, Integer=$User.Deployed))
       )
      )
     )
    )) {enabled=false, module=ScaleMeshSize003}
   /SetVariables(Particles.FeetPos:Vector3f=/Niagara/DynamicInputs/Transforms/TransformPosition(
     Position=&Add_Vector(
      A=Linked() {linkedVariable {name=Particles.Position, type=NiagaraPosition}}
      B=&MakeCustomVectorFromBool(Bool=$User.IsFrontEnd, True Vector=<0 -25 1500>, False Vector=<0 -25 15>)
     )
     Source Space=::Local
     Destination Space=::World
    )) {isSetParameters=true}
   /SetVariables(Particles.Feet2Pos:Vector3f=&TransformPosition(
     Position=&Add_Vector(
      A=Linked() {linkedVariable {name=Particles.Position, type=NiagaraPosition}}
      B=&MakeCustomVectorFromBool(Bool=$User.IsFrontEnd, True Vector=<0 37 1500>, False Vector=<0 37 15>)
     )
     Source Space=::Local
     Destination Space=::World
    )) {isSetParameters=true}
   &DynamicMaterialParameters(
    Index 0 Param 1=&Multiply_Float(
     A=&MakeFloatFromVector(Vector=$Particles.Scale, Channel=::NewEnumerator1)
     B=&MakeFloatFromVector(Vector=$Engine.Owner.Scale, Channel=::NewEnumerator0)
    )
    Index 0 Param 2=$Particles.FN_Timer.Value
    Index 0 Param 3=&Power(a=&MakeFloatFromVector(Vector=$Engine.Owner.Scale, Channel=::NewEnumerator0), b=1)
    Index 3 Param 2=&MakeCustomFloatFromBool(
     Bool=$User.IsFrontEnd
     True Float=0
     False Float=&FNTimeOfDayScalar(Night Scale=0.05, Morning Scale=-0.1, Day Scale=0, Evening Scale=-0.1)
    )
    Index 3 Param 3=&MakeCustomFloatFromBool(
     Bool=$User.IsFrontEnd
     True Float=1.6
     False Float=&FNTimeOfDayScalar(Night Scale=1, Morning Scale=0.3, Evening Scale=0.3)
    )
    Module.Param0WriteEnabled=true
    Module.Param0WriteEnabled3=true
    Module.Param1WriteEnabled=true
    Module.Param1WriteEnabled3=true
    Module.Param2WriteEnabled=true
    Module.Param2WriteEnabled3=true
    Module.Param3WriteEnabled=false
    Module.Param3WriteEnabled3=true
    Write Parameter Index 3=true
   )
   &DynamicMaterialParameters(
    Index 1 Param 1=&MakeFloatFromVector(Vector=$Particles.FeetPos, Channel=::NewEnumerator0)
    Index 1 Param 2=&MakeFloatFromVector(Vector=$Particles.FeetPos, Channel=::NewEnumerator1)
    Index 1 Param 3=&MakeFloatFromVector(Vector=$Particles.FeetPos, Channel=::NewEnumerator2)
    Index 1 Param 4=&FloatFromCurve(
     FloatCurve=DataInterface() {propertyValues {Curve {keys [<at 0 0>,<at 2 0>,<at 3 1>]}}}
     CurveIndex=&ClampFloat(
      Float=&Add_Float(A=$Particles.Age, B=&Multiply_FloatByInt(Float=100, Integer=$User.Deployed))
      Max=5
     )
    )
    Index 2 Param 1=&MakeFloatFromVector(Vector=$Particles.Feet2Pos, Channel=::NewEnumerator0)
    Index 2 Param 2=&MakeFloatFromVector(Vector=$Particles.Feet2Pos, Channel=::NewEnumerator1)
    Index 2 Param 3=&MakeFloatFromVector(Vector=$Particles.Feet2Pos, Channel=::NewEnumerator2)
    Module.Param0WriteEnabled1=true
    Module.Param0WriteEnabled2=true
    Module.Param1WriteEnabled1=true
    Module.Param1WriteEnabled2=true
    Module.Param2WriteEnabled1=true
    Module.Param2WriteEnabled2=true
    Module.Param3WriteEnabled1=true
    Module.Param3WriteEnabled2=false
    Write Parameter Index 0=false
    Write Parameter Index 1=true
    Write Parameter Index 2=true
   ) {module=DynamicMaterialParameters001}
  ]
  rendererClasses [&NiagaraMeshRendererProperties]
  renderers [
   {rendererClass=&NiagaraMeshRendererProperties, rendererIndex=0}
   {rendererClass=&NiagaraMeshRendererProperties, rendererIndex=1}
   {rendererClass=&NiagaraMeshRendererProperties, rendererIndex=2}
  ]
 }
 SpeedLine {
  simTarget=GPUComputeSim
  emitterUpdateScript [
   &EmitterState(
    Delay First Loop Only=true
    Inactive Response=::NewEnumerator1
    Life Cycle Mode=::NewEnumerator1
    Loop Delay=1.5
    Loop Duration=3
    UseLoopDelay=true
   )
   &SpawnRate(
    SpawnRate=&MakeCustomFloatFromBool(
     Bool=$User.IsFrontEnd
     True Float=100
     False Float=&Multiply_Float(
      A=&Multiply_Float(A=270, B=1)
      B=&Lerp_Float(A=-0.8, B=1, Alpha=$System.FN_Timer.Value)
     )
    )
    Use Spawn Probability=false
   )
  ]
  particleSpawnScript [
   &InitializeParticle(
    Lifetime Max=0.12
    Lifetime Min=0.1
    Lifetime Mode=::NewEnumerator1
    Position Offset=<50 0 0>
    Sprite Size Max=<2 7>
    Sprite Size Min=<2 2>
    Sprite Size Mode=::NewEnumerator2
    UsePositionOffset=true
    Write Material Random=true
   )
   &TorusLocation(
    Align Torus With Axis=true
    Handle Radius=0
    Large Radius:NiagaraFloat=&RandomRangeFloat(Minimum=&MakeCustomFloatFromBool(Bool=$User.IsFrontEnd, True Float=100, False Float=250), Maximum=400)
    MaskSpawnGroup=false
    Module.Override Randomness=false
    Offset=&FN_FrontEndVector3Modifier(Gameplay=<0 100 0>, Front End=<0 -100 0>)
    Torus Axis=<0 1 0>
   )
   &AddVelocity(Velocity=&RandomRangeVector(Minimum=<0 500 0>, Maximum=<0 500 0>))
   &AddVelocityFromPoint(Module.Use Falloff Distance=false, Origin Offset=<0 -200 0>, Velocity Strength=1500)
   &CameraOffset(Camera Offset Amount=-10)
  ]
  particleUpdateScript [
   &SolveForcesAndVelocity(Clamp Velocity=false, Limit Acceleration=false)
   &ParticleState()
   &ScaleSpriteSizeBySpeed(Max Scale Factor=<1 50>, Min Scale Factor=<0.01 0.05>, Velocity Threshold=1)
   &ScaleColor(
    Scale Alpha=&Multiply_Float(
     A=&FloatFromCurve(FloatCurve=DataInterface() {propertyValues {Curve {keys [
          {interpMode=RCIM_Cubic, tangentMode=RCTM_User, time=0, value=0}
          {interpMode=RCIM_Cubic, tangentMode=RCTM_User, time=0.5, value=1}
          {tangentMode=RCTM_User, time=1, value=0}
         ]}}})
     B=&Multiply_Float(
      A=&MakeCustomFloatFromBool(
       Bool=$User.IsFrontEnd
       True Float=&RandomRangeFloat(Minimum=0.1, Maximum=0.2)
       False Float=&RandomRangeFloat(Minimum=0.4, Maximum=0.6)
      )
      B=1
     )
    )
    ScaleA=true
    ScaleRGB=true
   )
  ]
  rendererClasses [&NiagaraSpriteRendererProperties]
  renderers [{rendererClass=&NiagaraSpriteRendererProperties}]
 }
}
summary {
 systemName=NS_GeniusBlob_Main
 emitters [
  {emitterName=NE_StaminaGliderCloud, rendererClasses [&NiagaraMeshRendererProperties]}
  {emitterName=SpeedLine, simTarget=GPUComputeSim, rendererClasses [&NiagaraSpriteRendererProperties]}
 ]
 userVariables [
  User.bIsFrontEnd {type=NiagaraBool}
  User.bIsFrontEndPreview {type=NiagaraBool}
  User.Deployed {type=NiagaraInt32}
  User.ForwardDot {type=NiagaraFloat, defaultValue=0, description=""}
  User.GlobalDelay {type=NiagaraFloat, defaultValue=0, description=""}
  User.IsFrontEnd {type=NiagaraBool}
  User.IsWhiteVersion {type=NiagaraBool}
  User.PMESH {
   type=NiagaraDataInterfaceSkeletalMesh
   defaultValue {
    struct=NiagaraExt_VariableValue_DataInterface
    value {dataInterface=&self:NiagaraDataInterfaceSkeletalMesh_0}
   }
  }
 ]
}
