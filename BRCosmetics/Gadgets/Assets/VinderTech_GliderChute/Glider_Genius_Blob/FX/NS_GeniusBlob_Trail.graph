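# Niagara trail for the Genius Blob glider, emitting cyan, outlined toon-shaded blob meshes.
# Two emitters launch particles that pop in with wobbling scale curves; User.ForwardDot drives spawn rate, lifetime and speed.
# The second emitter adds ForwardDot-paced X/Y squash and blends velocity-aligned, camera-rolled orientation toward camera-facing by view angle.
# User.IsFrontEnd switches start delay, base spawn rate and toon lighting colors for front-end display; User.RightDot is unused.
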NiagaraSystem /BRCosmetics/Gadgets/Assets/VinderTech_GliderChute/Glider_Genius_Blob/FX/NS_GeniusBlob_Trail
dependencies {
 usedDataInterfaces [&NiagaraDataInterfaceCurve, &NiagaraDataInterfaceCamera]
 usedDynamicInputs []
 usedModules [&SystemState]
}
emitters {
 Empty002 {
  emitterUpdateScript [
   &EmitterState(
    Inactive Response=::NewEnumerator1
    Life Cycle Mode=::NewEnumerator1
    Loop Behavior=::NewEnumerator1
    Loop Delay=&MakeCustomFloatFromBool(Bool=$User.IsFrontEnd, True Float=0.15, False Float=0.8)
    Loop Duration Mode=::NewEnumerator1
    UseLoopDelay=true
   )
   &SpawnRate(
    SpawnRate=&Lerp_Float(
     A=&MakeCustomFloatFromBool(Bool=$User.IsFrontEnd, True Float=8, False Float=4)
     B=8
     Alpha=$User.ForwardDot
    )
    Use Spawn Probability=false
   )
   /SetVariables(Emitter.FELight:LinearColor=&MakeCustomLinearColorFromBool(
     Bool=$User.IsFrontEnd
     True Color=&MakeLinearColorFromVectorAndFloat(
      "Vector (RGB)"=&NormalizeVector(Vector To Normalize=<0 1 1>, Safe Normalize=true, Safe Fallback Threshold=0, Safe Fallback Vector=<0 0 0.001>)
      "Float (Alpha)"=1
     )
     False Color=<rgba 0 0 0 0>
    )) {isSetParameters=true}
   /SetVariables(Emitter.TOD:Vector3f=&MakeCustomVectorFromBool(Bool=$User.IsFrontEnd, True Vector=<0 1 1>, False Vector=&MakeVector(X=0, Y=1, Z=1))) {isSetParameters=true}
   /SetVariables(Emitter.LitRim:LinearColor=&MakeCustomLinearColorFromBool(Bool=$User.IsFrontEnd, True Color=<rgba 0.25 1 1 1>, False Color=<rgba 0.2 0.8638 1 1>)) {isSetParameters=true}
   /SetVariables(Emitter.Lit:LinearColor=&MakeCustomLinearColorFromBool(Bool=$User.IsFrontEnd, True Color=<rgba 0.25 1 1 1>, False Color=<rgba 0.1 1 1 1>)) {isSetParameters=true}
   /SetVariables(Emitter.Shadow:LinearColor=&MakeCustomLinearColorFromBool(Bool=$User.IsFrontEnd, True Color=<rgba 0.12 0.5733 0.8 1>, False Color=<rgba 0.075 0.4292 0.5 1>)) {isSetParameters=true}
   /SetVariables(Emitter.GI:LinearColor=&MakeCustomLinearColorFromBool(Bool=$User.IsFrontEnd, True Color=<rgba 0.14 0.6067 0.7 1>, False Color=<rgba 0.0675 0.4181 0.45 1>)) {isSetParameters=true}
   /SetVariables(Emitter.SoftShadowStr:NiagaraFloat=&MakeCustomFloatFromBool(Bool=$User.IsFrontEnd, True Float=0.6, False Float=0.5)) {isSetParameters=true}
   /Game/Effects/Niagara/MasterMaterials/ToonShader/Template/NM_OutlineParameters(OutlineColor:LinearColor=<rgba 0 0.1866 0.224 1>)
  ]
  particleSpawnScript [
   &InitializeParticle(
    Color Mode=::NewEnumerator0
    Lifetime=&Multiply_Float(
     A=&RandomRangeFloat(Minimum=0.6, Maximum=0.8)
     B=&Lerp_Float(A=1.5, B=1, Alpha=$User.ForwardDot)
    )
    Mesh Scale=&Multiply_VectorByFloat(Vector=<1.5 1 1>, Float=&RandomRangeFloat(Minimum=0.01, Maximum=0.02))
    Mesh Scale Mode=::NewEnumerator4
    Position Offset=<0 20 -40>
    UsePositionOffset=true
    Write Material Random=true
   )
   &InitialMeshOrientation(
    Mesh Axis To Orient=<0 1 0>
    Mesh Orientation Mode=::NewEnumerator2
    Rotation=&RandomRangeVector(Minimum=<0 0 0>, Maximum=<360 0 0>)
    Use Rotation Vector=true
   )
   &AddVelocity(Velocity=&Multiply_VectorByFloat(
     Vector=&RandomRangeVector(Minimum=<0 300 50>, Maximum=<0 400 50>)
     Float=&Lerp_Float(A=0.65, B=1.1, Alpha=$User.ForwardDot)
    ))
   &ShapeLocation(MaskSpawnGroup=false, Non Uniform Scale=<1 0 0.35>, Override Seed=false, Sphere Radius=50, UseOwnerScale=true)
  ]
  particleUpdateScript [
   &ParticleState()
   &ScaleMeshSize(Scale Factor=&VectorFromFloat(Value=&FloatFromCurve(FloatCurve=DataInterface() {propertyValues {Curve {keys [
          {interpMode=RCIM_Cubic, leaveTangent=4.072, tangentMode=RCTM_Break, time=0, value=0}
          {arriveTangent=-1.3848, interpMode=RCIM_Cubic, leaveTangent=-1.3848, tangentMode=RCTM_User, time=0.3, value=1.5}
          {arriveTangent=-1.219, interpMode=RCIM_Cubic, leaveTangent=-1.219, tangentMode=RCTM_User, time=0.9, value=0.7}
          {arriveTangent=-6.6381, interpMode=RCIM_Cubic, tangentMode=RCTM_Break, time=1, value=0}
         ]}}}))) {enabled=false}
   &ScaleMeshSize(Scale Factor=&VectorFromFloat(Value=&FloatFromCurve(FloatCurve=DataInterface() {propertyValues {Curve {keys [
          {arriveTangent=16.6204, interpMode=RCIM_Cubic, leaveTangent=16.6204, tangentMode=RCTM_User, time=0, value=0}
          {arriveTangent=0.2654, interpMode=RCIM_Cubic, leaveTangent=0.2654, tangentMode=RCTM_User, time=0.1636, value=1.476}
          {arriveTangent=-0.7988, interpMode=RCIM_Cubic, leaveTangent=-0.7988, time=0.2959, value=0.814}
          {arriveTangent=0.3306, interpMode=RCIM_Cubic, leaveTangent=0.3306, time=0.4349, value=1.2592}
          {arriveTangent=-0.8945, interpMode=RCIM_Cubic, leaveTangent=-0.8945, time=0.5567, value=0.9002}
          {arriveTangent=-0.5833, interpMode=RCIM_Cubic, leaveTangent=-0.5833, time=0.7325, value=0.993}
          {arriveTangent=-3.7129, interpMode=RCIM_Cubic, leaveTangent=-3.7129, time=0.9, value=0.7}
          {arriveTangent=-3.986, interpMode=RCIM_Cubic, leaveTangent=-3.986, tangentMode=RCTM_User, time=1, value=0}
         ]}}}))) {module=ScaleMeshSize001}
   &DynamicMaterialParameters(
    Index 0 Param 1=&MakeFloatFromVector(Vector=$Particles.Scale, Channel=::NewEnumerator1)
    Index 0 Param 2=&RandomRangeFloat(Minimum=0, Maximum=1000)
    Index 0 Param 3=0.7
    Module.Param0WriteEnabled=true
    Module.Param0WriteEnabled1=true
    Module.Param1WriteEnabled=true
    Module.Param1WriteEnabled1=true
    Module.Param2WriteEnabled=true
    Module.Param2WriteEnabled1=true
    Module.Param3WriteEnabled=true
    Module.Param3WriteEnabled1=true
    Write Parameter Index 1=true
   ) {module=DynamicMaterialParameters001}
   &Drag(Drag=0.4, Use Linear Drag=true, Use Rotational Drag=false)
   &SolveForcesAndVelocity(Clamp Velocity=false, Limit Acceleration=false)
  ]
  rendererClasses [&NiagaraSpriteRendererProperties, &NiagaraMeshRendererProperties]
  renderers [
   {rendererClass=&NiagaraSpriteRendererProperties}
   {rendererClass=&NiagaraMeshRendererProperties, rendererIndex=1}
   {rendererClass=&NiagaraMeshRendererProperties, rendererIndex=2}
  ]
 }
 Empty003 {
  emitterUpdateScript [
   &EmitterState(
    Inactive Response=::NewEnumerator1
    Life Cycle Mode=::NewEnumerator1
    Loop Behavior=::NewEnumerator1
    Loop Delay=&MakeCustomFloatFromBool(Bool=$User.IsFrontEnd, True Float=0.1, False Float=0.6)
    Loop Duration Mode=::NewEnumerator1
    UseLoopDelay=true
   )
   &SpawnRate(
    SpawnRate=&Lerp_Float(
     A=&MakeCustomFloatFromBool(Bool=$User.IsFrontEnd, True Float=4.5, False Float=3)
     B=7
     Alpha=$User.ForwardDot
    )
    Use Spawn Probability=false
   )
   /SetVariables(Emitter.FELight:LinearColor=&MakeCustomLinearColorFromBool(
     Bool=$User.IsFrontEnd
     True Color=&MakeLinearColorFromVectorAndFloat(
      "Vector (RGB)"=&NormalizeVector(Vector To Normalize=<0 1 1>, Safe Normalize=true, Safe Fallback Threshold=0, Safe Fallback Vector=<0 0 0.001>)
      "Float (Alpha)"=1
     )
     False Color=<rgba 0 0 0 0>
    )) {isSetParameters=true}
   /SetVariables(Emitter.TOD:Vector3f=&MakeCustomVectorFromBool(Bool=$User.IsFrontEnd, True Vector=<0 1 1>, False Vector=&MakeVector(X=0, Y=1, Z=1))) {isSetParameters=true}
   /SetVariables(Emitter.LitRim:LinearColor=&MakeCustomLinearColorFromBool(Bool=$User.IsFrontEnd, True Color=<rgba 0.25 1 1 1>, False Color=<rgba 0.2 0.8638 1 1>)) {isSetParameters=true}
   /SetVariables(Emitter.Lit:LinearColor=&MakeCustomLinearColorFromBool(Bool=$User.IsFrontEnd, True Color=<rgba 0.25 1 1 1>, False Color=<rgba 0.1 1 1 1>)) {isSetParameters=true}
   /SetVariables(Emitter.Shadow:LinearColor=&MakeCustomLinearColorFromBool(Bool=$User.IsFrontEnd, True Color=<rgba 0.12 0.5733 0.8 1>, False Color=<rgba 0.075 0.4292 0.5 1>)) {isSetParameters=true}
   /SetVariables(Emitter.GI:LinearColor=&MakeCustomLinearColorFromBool(Bool=$User.IsFrontEnd, True Color=<rgba 0.14 0.6067 0.7 1>, False Color=<rgba 0.0675 0.4181 0.45 1>)) {isSetParameters=true}
   /SetVariables(Emitter.SoftShadowStr:NiagaraFloat=&MakeCustomFloatFromBool(Bool=$User.IsFrontEnd, True Float=0.6, False Float=0.5)) {isSetParameters=true}
   &NM_OutlineParameters(OutlineColor:LinearColor=<rgba 0 0.1866 0.224 1>)
  ]
  particleSpawnScript [
   &InitializeParticle(
    Color Mode=::NewEnumerator0
    Lifetime=&Multiply_Float(
     A=&RandomRangeFloat(Minimum=0.6, Maximum=0.8)
     B=&Lerp_Float(A=2, B=0.7, Alpha=$User.ForwardDot)
    )
    Mesh Scale=&Multiply_VectorByFloat(Vector=<0.85 1.23 1>, Float=&RandomRangeFloat(Minimum=0.6, Maximum=0.85))
    Mesh Scale Mode=::NewEnumerator4
    Position Offset=<0 30 -40>
    UsePositionOffset=true
    Write Material Random=true
   )
   &AddVelocity(Velocity=&Multiply_VectorByFloat(
     Vector=&RandomRangeVector(Minimum=<0 250 50>, Maximum=<0 320 50>)
     Float=&Lerp_Float(A=0.5, B=2, Alpha=$User.ForwardDot)
    ))
   &ShapeLocation(
    MaskSpawnGroup=false
    Non Uniform Scale=<1 0 0.6>
    Override Seed=false
    Sphere Radius=80
    Sphere Surface Distribution=0.8
    UseOwnerScale=true
   )
  ]
  particleUpdateScript [
   &ParticleState()
   &ScaleMeshSize(Scale Factor=&VectorFromFloat(Value=&FloatFromCurve(FloatCurve=DataInterface() {propertyValues {Curve {keys [
          {interpMode=RCIM_Cubic, leaveTangent=4.072, tangentMode=RCTM_Break, time=0, value=0}
          {arriveTangent=-1.3848, interpMode=RCIM_Cubic, leaveTangent=-1.3848, tangentMode=RCTM_User, time=0.3, value=1.5}
          {arriveTangent=-1.219, interpMode=RCIM_Cubic, leaveTangent=-1.219, tangentMode=RCTM_User, time=0.9, value=0.7}
          {arriveTangent=-6.6381, interpMode=RCIM_Cubic, tangentMode=RCTM_Break, time=1, value=0}
         ]}}}))) {enabled=false}
   &ScaleMeshSize(Scale Factor=&VectorFromFloat(Value=&FloatFromCurve(FloatCurve=DataInterface() {propertyValues {Curve {keys [
          {arriveTangent=14.4556, interpMode=RCIM_Cubic, leaveTangent=14.4556, tangentMode=RCTM_User, time=0, value=0}
          {arriveTangent=3.802, interpMode=RCIM_Cubic, leaveTangent=3.802, time=0.1439, value=1.6621}
          {arriveTangent=-1.0723, interpMode=RCIM_Cubic, leaveTangent=-1.0723, time=0.2814, value=1.0699}
          {arriveTangent=0.1453, interpMode=RCIM_Cubic, leaveTangent=0.1453, time=0.4184, value=1.3678}
          {arriveTangent=-0.4845, interpMode=RCIM_Cubic, leaveTangent=-0.4845, time=0.5284, value=1.1058}
          {arriveTangent=-1.0919, interpMode=RCIM_Cubic, leaveTangent=-1.0919, time=0.6797, value=1.2412}
          {arriveTangent=-3.875, interpMode=RCIM_Cubic, leaveTangent=-3.875, time=0.9, value=0.7}
          {arriveTangent=-3.986, interpMode=RCIM_Cubic, leaveTangent=-3.986, tangentMode=RCTM_User, time=1, value=0}
         ]}}}))) {module=ScaleMeshSize001}
   &ScaleMeshSize(Scale Factor=&MakeVector(
     X=&FloatFromCurve(
      FloatCurve=DataInterface() {propertyValues {Curve {keys [
          <at 0 1.0054 RCIM_Cubic>
          {arriveTangent=-4.6274, interpMode=RCIM_Cubic, leaveTangent=-4.6274, time=0.0704, value=1.1558}
          {arriveTangent=1.6302, interpMode=RCIM_Cubic, leaveTangent=1.6302, time=0.1634, value=0.2494}
          {arriveTangent=0.6656, interpMode=RCIM_Cubic, leaveTangent=0.6656, time=0.2672, value=1.4767}
          {arriveTangent=-0.4902, interpMode=RCIM_Cubic, leaveTangent=-0.4902, time=0.3727, value=0.3888}
          {arriveTangent=0.7676, interpMode=RCIM_Cubic, leaveTangent=0.7676, time=0.5031, value=1.3611}
          {arriveTangent=-0.9262, interpMode=RCIM_Cubic, leaveTangent=-0.9262, time=0.6444, value=0.5973}
          {arriveTangent=1.3948, interpMode=RCIM_Cubic, leaveTangent=1.3948, time=0.79, value=1.0953}
          <at 0.9064 0.9627>
          <at 1 1 RCIM_Cubic>
         ]}}}
      CurveIndex=&Lerp_Float(A=&Multiply_Float(A=$Particles.NormalizedAge, B=2), B=$Particles.NormalizedAge, Alpha=$User.ForwardDot)
     )
     Y=&FloatFromCurve(
      FloatCurve=DataInterface() {propertyValues {Curve {keys [
          <at 0 1 RCIM_Cubic>
          {arriveTangent=2.312, interpMode=RCIM_Cubic, leaveTangent=2.312, time=0.1116, value=0.7866}
          {arriveTangent=-2.3284, interpMode=RCIM_Cubic, leaveTangent=-2.3284, time=0.2078, value=1.4805}
          {arriveTangent=-1.5652, interpMode=RCIM_Cubic, leaveTangent=-1.5652, time=0.3037, value=0.3395}
          {arriveTangent=1.0967, interpMode=RCIM_Cubic, leaveTangent=1.0967, time=0.4218, value=1.1457}
          {arriveTangent=-0.1955, interpMode=RCIM_Cubic, leaveTangent=-0.1955, time=0.5108, value=0.5667}
          {arriveTangent=1.7316, interpMode=RCIM_Cubic, leaveTangent=1.7316, time=0.6069, value=1.1095}
          {arriveTangent=-0.235, interpMode=RCIM_Cubic, leaveTangent=-0.235, time=0.7008, value=0.8955}
          {arriveTangent=0.3491, interpMode=RCIM_Cubic, leaveTangent=0.3491, time=0.8469, value=1.0531}
          <at 1 1 RCIM_Cubic>
         ]}}}
      CurveIndex=&Lerp_Float(A=&Multiply_Float(A=$Particles.NormalizedAge, B=2), B=$Particles.NormalizedAge, Alpha=$User.ForwardDot)
     )
     Z=1
    )) {module=ScaleMeshSize002}
   &DynamicMaterialParameters(
    Index 0 Param 1=&MakeFloatFromVector(Vector=$Particles.Scale, Channel=::NewEnumerator1)
    Index 0 Param 2=&RandomRangeFloat(Minimum=0, Maximum=1000)
    Index 0 Param 3=0.7
    Module.Param0WriteEnabled=true
    Module.Param0WriteEnabled1=true
    Module.Param1WriteEnabled=true
    Module.Param1WriteEnabled1=true
    Module.Param2WriteEnabled=true
    Module.Param2WriteEnabled1=true
    Module.Param3WriteEnabled=true
    Module.Param3WriteEnabled1=true
    Write Parameter Index 1=true
   ) {module=DynamicMaterialParameters001}
   &Drag(Drag=0.4, Use Linear Drag=true, Use Rotational Drag=false)
   &SolveForcesAndVelocity(Clamp Velocity=false, Limit Acceleration=false)
   /SetVariables(Particles.MeshOrientation=&LerpQuaternion(
     Quaternion A:Quat4f=&MultiplyQuaternion(
      Quaternion A=&FindQuatBetweenTwoVectors(
       From Vector=<0 1 0>
       To Vector=&NormalizeVector(Vector To Normalize=$Particles.Velocity, Safe Normalize=true, Safe Fallback Threshold=0, Safe Fallback Vector=<0 0 0.001>)
      )
      Quaternion B=&FindQuatBetweenTwoVectors(
       From Vector=<0 0 1>
       To Vector=&self:ScratchDynamicInput(
        CameraPosition:NiagaraPosition=&ReturnCameraPosition(Coordinate Space:ENiagaraCoordinateSpace=::Simulation, Camera Query:NiagaraDataInterfaceCamera=DataInterface())
        RotAxis:Vector3f=<0 1 0>
       )
      )
     )
     Quaternion B:Quat4f=&FindQuatBetweenTwoVectors(
      From Vector=<0 0 1>
      To Vector=&Subtract_Position(
       A=&ReturnCameraPosition(Coordinate Space:ENiagaraCoordinateSpace=::Simulation, Camera Query:NiagaraDataInterfaceCamera=DataInterface())
       B=$Particles.Position
      )
     )
     Lerp Factor:NiagaraFloat=&RemapRange(
      Input Value To Remap=&self:ScratchDynamicInput_02(
       Axis:Vector3f=&NormalizeVector(Vector To Normalize=$Particles.Velocity, Safe Normalize=true, Safe Fallback Threshold=0, Safe Fallback Vector=<0 0 0.001>)
       CameraPos:NiagaraPosition=&ReturnCameraPosition(Coordinate Space:ENiagaraCoordinateSpace=::Simulation, Camera Query:NiagaraDataInterfaceCamera=DataInterface())
       Pow:NiagaraFloat=2
      )
      Input Min=0
      Input Max=1
      Output Min=0
      Output Max=0.4
     )
     Enable Spherical Linear Interpolation:NiagaraBool=false
    )) {isSetParameters=true}
  ]
  rendererClasses [&NiagaraSpriteRendererProperties, &NiagaraMeshRendererProperties]
  renderers [
   {rendererClass=&NiagaraSpriteRendererProperties}
   {rendererClass=&NiagaraMeshRendererProperties, rendererIndex=1}
   {rendererClass=&NiagaraMeshRendererProperties, rendererIndex=2}
  ]
 }
}
summary {
 systemName=NS_GeniusBlob_Trail
 emitters [
  {emitterName=Empty002, rendererClasses [&NiagaraSpriteRendererProperties, &NiagaraMeshRendererProperties]}
  {emitterName=Empty003, rendererClasses [&NiagaraSpriteRendererProperties, &NiagaraMeshRendererProperties]}
 ]
 userVariables [
  User.ForwardDot {type=NiagaraFloat, defaultValue=0, description=""}
  User.IsFrontEnd {type=NiagaraBool, defaultValue=true}
  User.RightDot {type=NiagaraFloat, defaultValue=0, description=""}
 ]
}
