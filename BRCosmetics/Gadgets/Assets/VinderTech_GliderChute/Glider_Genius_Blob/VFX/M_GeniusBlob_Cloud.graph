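# Cloth-shaded cloud material for Niagara mesh particles in the Genius Blob glider VFX.
# A shared CloudWPO function builds the cloud from particle position, dynamic-parameter scale, age and grow channels, a UV-gradient mask and two foot-hole positions, each built from its X channel alone.
# Its normal feeds toon shading lit by the sky atmosphere direction or a FELightDir override.
# HSV, optional particle color and emissive adjustments then tint the result.
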Material /BRCosmetics/Gadgets/Assets/VinderTech_GliderChute/Glider_Genius_Blob/VFX/M_GeniusBlob_Cloud
graph {
 properties {bUsedWithNiagaraMeshParticles=true, pixelDepthOffsetMode=PDOM_Legacy, shadingModel=MSM_Cloth, usedShadingModels=Cloth}
 nodes [
  ActorPositionWS()
  n1 = ScalarParameter() {defaultValue=1, parameterName=Opacity}
  n2 = ParticleColor()
  LinearInterpolate(A=@n1, B=Multiply(A=@n1, B=@n2.A), Alpha=ScalarParameter() {parameterName=Use Particle ALpha})
  n3 = DynamicParameter() {paramNames [ScaleX,Age,Param3,Param4]}
  n4 = VectorParameter() {parameterName=TempCloudCenterOffset, defaultValue=<rgba 0 0 0 1>}
  n5 = MaterialFunctionCall(Float2=TextureCoordinate()) {materialFunction=&BreakOutFloat2Components}
  n6 = Power(Base=OneMinus(None=OneMinus(None=@n5.G)), Exp=ScalarParameter() {defaultValue=4, parameterName=aaa})
  n7 = Constant() {r=1}
  n8 = DynamicParameter() {parameterIndex=1, defaultValue=<rgba 1 1 1 0>, paramNames [FeetPosX, FeetPosY, FeetPosZ, FeetHoleColor]}
  n9 = DynamicParameter() {parameterIndex=2, paramNames [Feet2PosX,Feet2PosY,Feet2PosZ,Grow]}
  n10 = MaterialFunctionCall!(
   CloudUV_Main
   CloudScale=@n3.ScaleX
   CloudUV_Detail=TextureCoordinate() {coordinateIndex=1}
   FlatMaskCenterPos=Add(A=ParticlePositionWS(), B=Transform(None=@n4.RGB) {transformSourceType=TRANSFORMSOURCE_Local})
   Mask=MaterialFunctionCall(
    Input=OneMinus(None=@n6)
    Input Low=Constant()
    Input High=@n7
    Target Low=ScalarParameter() {defaultValue=0.3, parameterName=bbb}
    Target High=@n7
   ) {materialFunction=&RemapValueRange}.Result
   FootHole_Left_Position=MaterialFunctionCall(X=@n8.FeetPosX, Y=@n8.FeetPosX, Z=@n8.FeetPosX) {materialFunction=&MakeFloat3}.Result
   FootHole_Right_Position=MaterialFunctionCall(X=@n9.Feet2PosX, Y=@n9.Feet2PosX, Z=@n9.Feet2PosX) {materialFunction=&MakeFloat3}.Result
   TimeIn=Add(A=Time(), B=@n3.Age)
  ) {materialFunction=&/FXCommon_CloudWPO}
  MaterialFunctionCall(Normal=@n10.NormalTS, Flatness=ScalarParameter() {defaultValue=-1, parameterName=F}) {materialFunction=&FlattenNormal}
  MaterialFunctionCall() {materialFunction=&ObjectScale}
  n11 = VectorParameter() {parameterName=FELightDir, defaultValue=<rgba 0 0 0 1>}
  n12 = Abs(None=Add(A=Add(A=@n11.R, B=@n11.R), B=@n11.B))
  n13 = LinearInterpolate(A=SkyAtmosphereLightDirection(), B=@n11.RGB, Alpha=Saturate(None=Ceil(None=@n12)))
  n14 = MaterialFunctionCall!(
   HSV=VectorParameter() {parameterName=HSV, defaultValue=<rgba 0 1 1 1>}.RGB
   Color=LinearInterpolate(
    A=MaterialFunctionCall!(LightDirIn=@n13, Masks, PixelNormalTS=@n10.NormalTS) {materialFunction=&/FXCommon_CloudToonShading}.Emmissive
    B=@n10.FlatMask
    Alpha=ScalarParameter() {parameterName=Deb}
   )
  ) {materialFunction=/Game/Effects/Niagara/MasterMaterials/Functions/Common/FXCommon_HSV}
  MaterialFunctionCall(Emissive=LinearInterpolate(A=@n14.Result, B=Multiply(A=@n14.Result, B=@n2.RGB), Alpha=ScalarParameter() {parameterName=use Particle COlor})) {materialFunction=&MF_FX_EmissiveAdjustments}
  n15 = VertexNormalWS()
  n16 = CameraVectorWS()
  OneMinus(None=Ceil(None=DotProduct(A=@n15, B=@n16)))
  ScalarParameter() {parameterName=Metallic}
  ScalarParameter() {defaultValue=0.969, parameterName=Roughness}
  ScalarParameter() {parameterName=Spec}
  VectorParameter() {parameterName=BaseColor, defaultValue=<rgba 0.502 0.502 0.502 0>}
  VectorParameter() {parameterName=FuzzColor, defaultValue=<rgba 0.003 0.001 0.002 1>}
 ]
}
